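# Animation Blueprint for the JazzShoes pickaxe skeleton that plays the weapon's chain animation.
# The Output Pose takes a sequence player directly, leaving the Chain state machine unconnected.
# That machine switches instantly between an Idle pose evaluated at time zero and a playing Spin state, driven by MeleeAnimPlaying.
# Each update sets that flag from the MeleeAnim curve on the owning player's anim instance, reached through the pickaxe actor's owner.
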AnimBlueprint /BRCosmetics/Weapons/FORT_Melee/Pickaxe_JazzShoes/Meshes/JazzShoes_Axe_AnimBP
target_skeleton=&/JazzShoes_Axe_Skeleton
graphs [
 AnimGraph {nodes [
   Output Pose(Result=Sequence Player 'JazzShoes_Chain_WEP'() {sequence=/BRCosmetics/Weapons/FORT_Melee/Pickaxe_JazzShoes/Anim/JazzShoes_Chain_WEP}) {layerGroup=None}
   AnimGraphNode_StateMachine_0 = Chain() {is=title AnimGraphNode_StateMachine}
  ]}
 Chain {
  entry=Idle
  path=AnimGraph.AnimGraphNode_StateMachine_0.Chain
  ids=id
  nodes [
   AnimStateNode_0 = Spin() {is=title AnimStateNode}
   AnimStateNode_1 = Idle()
   AnimStateTransitionNode_0 = AnimStateTransitionNode() {blendMode=Linear, crossfadeDuration=0}
   AnimStateTransitionNode_1 = AnimStateTransitionNode() {blendMode=Linear, crossfadeDuration=0}
  ]
  links [
   {dst=Spin, rule=AnimGraph.AnimGraphNode_StateMachine_0.Chain.AnimStateTransitionNode_0.Transition, src=Idle, via=AnimStateTransitionNode_0}
   {dst=Idle, rule=AnimGraph.AnimGraphNode_StateMachine_0.Chain.AnimStateTransitionNode_1.Transition, src=Spin, via=AnimStateTransitionNode_1}
  ]
 }
 Spin {
  path=AnimGraph.AnimGraphNode_StateMachine_0.Chain.AnimStateNode_0.Spin
  nodes [Output Animation Pose(Result=Sequence Player 'JazzShoes_Chain_WEP'() {sequence=&JazzShoes_Chain_WEP})]
 }
 Idle {
  path=AnimGraph.AnimGraphNode_StateMachine_0.Chain.AnimStateNode_1.Idle
  nodes [Output Animation Pose(Result="Sequence Evaluator (by time) 'JazzShoes_Chain_WEP'"!(ExplicitTime:float=0) -> (Pose:Pose Link Structure) {is=title AnimGraphNode_SequenceEvaluator, sequence=&JazzShoes_Chain_WEP})]
 }
 Transition {
  path=AnimGraph.AnimGraphNode_StateMachine_0.Chain.AnimStateTransitionNode_0.Transition
  nodes [Result(bCanEnterTransition=Get MeleeAnimPlaying() -> (MeleeAnimPlaying:bool))]
 }
 Transition {
  path=AnimGraph.AnimGraphNode_StateMachine_0.Chain.AnimStateTransitionNode_1.Transition
  nodes [Result(bCanEnterTransition=NOT Boolean(A=Get MeleeAnimPlaying()))]
 }
 EventGraph {nodes [
   Event BlueprintUpdateAnimation()
   > n1 = Cast To FortWeaponPickaxeAthena!(execute, Object=GetOwningActor()) -> (then, CastFailed, AsFort Weapon Pickaxe Athena:Fort Weapon Pickaxe Athena Object Reference) {is=title K2Node_DynamicCast}
   n2 = GetOwner(self:Actor Object Reference=@n1)
   > Is Valid(InputObject=@n2)
   n3 = GetComponentByClass!(self:Actor Object Reference=@n2, ComponentClass:Actor Component Class Reference=/Script/Engine.SkeletalMeshComponent) -> (ReturnValue:Skeletal Mesh Component Object Reference)
   > Is Valid(InputObject=@n3)
   > n4 = Cast To FortPlayerAnimInstance(Object=GetAnimInstance(self=@n3))
   > Set MeleeAnimPlaying(MeleeAnimPlaying:bool="float >= float"(A=GetCurveValue(self=@n4, CurveName=MeleeAnim), B=1)) -> (Output_Get:bool)
   TryGetPawnOwner()
  ]}
]
variables [MeleeAnimPlaying {type=boolean}]
